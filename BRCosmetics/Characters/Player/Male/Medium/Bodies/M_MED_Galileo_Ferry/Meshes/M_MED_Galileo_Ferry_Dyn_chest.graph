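# Physics asset for the dynamic chest flap on the M_MED_Galileo_Ferry medium male player body.
# A small capsule on the master flap bone anchors a two-segment chain of thin box bodies.
# Constraints link the first segment to the master and the second to the first, with limited swing at both joints.
# Twist is locked at the root joint and tightly limited between the segments.
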PhysicsAsset /BRCosmetics/Characters/Player/Male/Medium/Bodies/M_MED_Galileo_Ferry/Meshes/M_MED_Galileo_Ferry_Dyn_chest
bodies {
 dyn_chest_flap_01 [{shapeType=Box, extentX=4.1875, extentY=6.875, extentZ=1.1875, center=<2.1509 -0.0001 0.0003>}]
 dyn_chest_flap_02 [{shapeType=Box, extentX=3.5625, extentY=6.875, extentZ=1.1875, center=<1.8301 -0.0001 0.0007>}]
 dyn_chest_flap_MSTR [{shapeType=Capsule, length=1, radius=1}]
}
constraints [
 {
  bone1Name=dyn_chest_flap_01
  bone2Name=dyn_chest_flap_MSTR
  swing1LimitDegrees=10
  swing2LimitDegrees=25
  twistLimitDegrees=45
  twistMotion=Locked
 }
 {bone1Name=dyn_chest_flap_02, bone2Name=dyn_chest_flap_01, swing1LimitDegrees=10, swing2LimitDegrees=25, twistLimitDegrees=2}
]
