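# Physics asset for the dynamic jacket on the M_MED_Galileo_Ferry character body, named as the skydiving variant.
# Seven three-box chains form the panels: three back chains hang from spine_04, side and front chains from MSTR capsules.
# Back and side joints bend on one swing axis, front joints on limited swing and twist.
# Free-angular constraints tie neighbouring bottom segments around to both fronts, while capsules cover pelvis, spine and thighs.
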PhysicsAsset /BRCosmetics/Characters/Player/Male/Medium/Bodies/M_MED_Galileo_Ferry/Meshes/M_MED_Galileo_Ferry_Dyn_jacket_skydiving
bodies {
 dyn_jacket_bk_01 [{shapeType=Box, extentX=9.375, extentY=15.125, extentZ=1.1875, center=<4.3573 0 0.0002>}]
 dyn_jacket_bk_02 [{shapeType=Box, extentX=9.375, extentY=15.125, extentZ=1.1875, center=<4.3573 0 0.0002>}]
 dyn_jacket_bk_03 [{shapeType=Box, extentX=9.375, extentY=15.125, extentZ=1.1875, center=<4.3573 0 0.0002>}]
 dyn_jacket_bk_l_01 [{shapeType=Box, extentX=9.375, extentY=15.125, extentZ=1.1875, center=<4.3573 0 0.0002>}]
 dyn_jacket_bk_l_02 [{shapeType=Box, extentX=9.375, extentY=15.125, extentZ=1.1875, center=<4.3573 0 0.0002>}]
 dyn_jacket_bk_l_03 [{shapeType=Box, extentX=9.375, extentY=15.125, extentZ=1.1875, center=<4.3573 0 0.0002>, rotation=<rot 0 0 7.8077>}]
 dyn_jacket_bk_r_01 [{shapeType=Box, extentX=9.375, extentY=15.125, extentZ=1.1875, center=<4.3573 0 0.0002>}]
 dyn_jacket_bk_r_02 [{shapeType=Box, extentX=9.375, extentY=15.125, extentZ=1.1875, center=<4.3573 0 0.0002>, rotation=<rot 0 0 -7.3051>}]
 dyn_jacket_bk_r_03 [{shapeType=Box, extentX=9.375, extentY=15.125, extentZ=1.1875, center=<4.3573 0 0.0002>, rotation=<rot 0 0 -6.3335>}]
 dyn_jacket_fr_l_01 [{shapeType=Box, extentX=9.375, extentY=15.125, extentZ=1.1875, center=<4.3573 0 0.0002>}]
 dyn_jacket_fr_l_02 [{shapeType=Box, extentX=9.375, extentY=13.4375, extentZ=1.1875, center=<4.3572 -1.5 0.0002>, rotation=<rot 0 0 6.2944>}]
 dyn_jacket_fr_l_03 [{shapeType=Box, extentX=9.375, extentY=13.4375, extentZ=1.1875, center=<4.3572 -1.5 0.0002>, rotation=<rot 0 0 10.1176>}]
 dyn_jacket_fr_l_MSTR [{shapeType=Capsule, length=1, radius=1}]
 dyn_jacket_fr_r_01 [{shapeType=Box, extentX=9.375, extentY=15.125, extentZ=1.1875, center=<4.3576 2 0.0002>, rotation=<rot 0 0 -8.893>}]
 dyn_jacket_fr_r_02 [{shapeType=Box, extentX=9.375, extentY=15.125, extentZ=1.1875, center=<4.3576 2 0.0002>, rotation=<rot 0 0 -8.893>}]
 dyn_jacket_fr_r_03 [{shapeType=Box, extentX=9.375, extentY=15.125, extentZ=1.1875, center=<4.3576 2 0.0002>, rotation=<rot 0 0 -9.7322>}]
 dyn_jacket_fr_r_MSTR [{shapeType=Capsule, length=1, radius=1}]
 dyn_jacket_l_01 [{shapeType=Box, extentX=9.375, extentY=15.125, extentZ=1.1875, center=<4.3573 0 0.0002>}]
 dyn_jacket_l_02 [{shapeType=Box, extentX=9.375, extentY=15.125, extentZ=1.1875, center=<4.3573 0 0.0002>}]
 dyn_jacket_l_03 [{shapeType=Box, extentX=9.375, extentY=15.125, extentZ=1.1875, center=<4.3573 0 0.0002>}]
 dyn_jacket_l_MSTR [{shapeType=Capsule, length=1, radius=1}]
 dyn_jacket_r_01 [{shapeType=Box, extentX=9.375, extentY=15.125, extentZ=1.1875, center=<4.3573 0 0.0002>}]
 dyn_jacket_r_02 [{shapeType=Box, extentX=9.375, extentY=15.125, extentZ=1.1875, center=<4.3573 0 0.0002>}]
 dyn_jacket_r_03 [{shapeType=Box, extentX=9.375, extentY=15.125, extentZ=1.1875, center=<4.3573 0 0.0002>}]
 dyn_jacket_r_MSTR [{shapeType=Capsule, length=1, radius=1}]
 pelvis [{shapeType=Capsule, length=11.375, radius=11.9688, center=<-1.6591 -1.9342 0>}]
 spine_01 [{shapeType=Capsule, length=12.2188, radius=10.3438, center=<0 -2.6299 -0.0001>}]
 spine_02 [{shapeType=Capsule, length=10.0938, radius=10.9062, center=<0 -1.3574 0>}]
 spine_03 [{shapeType=Capsule, length=11.2812, radius=10, center=<0 -0.4524 0>}]
 spine_04 [{shapeType=Capsule, length=1, radius=1}]
 thigh_l [{shapeType=Capsule, length=29.3438, radius=9.4375, center=<-18.4382 0.0005 0.656>, rotation=<rot 85.5514 0 0>}]
 thigh_r [{shapeType=Capsule, length=29.3125, radius=9.4688, center=<18.4892 0.0004 0>, rotation=<rot 90 0 0>}]
}
constraints [
 {bone1Name=dyn_jacket_fr_l_02, bone2Name=dyn_jacket_fr_l_01, swing1LimitDegrees=45, swing2LimitDegrees=60, twistLimitDegrees=10}
 {bone1Name=dyn_jacket_fr_l_03, bone2Name=dyn_jacket_fr_l_02, swing1LimitDegrees=45, swing2LimitDegrees=45, twistLimitDegrees=15}
 {bone1Name=dyn_jacket_fr_l_01, bone2Name=dyn_jacket_fr_l_MSTR, swing1LimitDegrees=45, swing2LimitDegrees=60, twistLimitDegrees=5}
 {
  bone1Name=dyn_jacket_l_01
  bone2Name=dyn_jacket_l_MSTR
  swing1LimitDegrees=45
  swing1Motion=Locked
  swing2LimitDegrees=45
  twistLimitDegrees=5
  twistMotion=Locked
 }
 {
  bone1Name=dyn_jacket_l_02
  bone2Name=dyn_jacket_l_01
  swing1LimitDegrees=45
  swing1Motion=Locked
  swing2LimitDegrees=45
  twistLimitDegrees=5
  twistMotion=Locked
 }
 {
  bone1Name=dyn_jacket_l_03
  bone2Name=dyn_jacket_l_02
  swing1LimitDegrees=45
  swing1Motion=Locked
  swing2LimitDegrees=45
  twistLimitDegrees=5
  twistMotion=Locked
 }
 {
  bone1Name=dyn_jacket_bk_l_01
  bone2Name=spine_04
  swing1LimitDegrees=0
  swing1Motion=Locked
  swing2LimitDegrees=45
  twistLimitDegrees=5
  twistMotion=Locked
 }
 {
  bone1Name=dyn_jacket_bk_l_02
  bone2Name=dyn_jacket_bk_l_01
  swing1LimitDegrees=0
  swing1Motion=Locked
  swing2LimitDegrees=45
  twistLimitDegrees=5
  twistMotion=Locked
 }
 {
  bone1Name=dyn_jacket_bk_l_03
  bone2Name=dyn_jacket_bk_l_02
  swing1LimitDegrees=0
  swing1Motion=Locked
  swing2LimitDegrees=45
  twistLimitDegrees=5
  twistMotion=Locked
 }
 {
  bone1Name=dyn_jacket_bk_01
  bone2Name=spine_04
  swing1LimitDegrees=0
  swing1Motion=Locked
  swing2LimitDegrees=45
  twistLimitDegrees=5
  twistMotion=Locked
 }
 {
  bone1Name=dyn_jacket_bk_02
  bone2Name=dyn_jacket_bk_01
  swing1LimitDegrees=0
  swing1Motion=Locked
  swing2LimitDegrees=45
  twistLimitDegrees=5
  twistMotion=Locked
 }
 {
  bone1Name=dyn_jacket_bk_03
  bone2Name=dyn_jacket_bk_02
  swing1LimitDegrees=0
  swing1Motion=Locked
  swing2LimitDegrees=45
  twistLimitDegrees=5
  twistMotion=Locked
 }
 {
  bone1Name=dyn_jacket_bk_r_01
  bone2Name=spine_04
  swing1LimitDegrees=0
  swing1Motion=Locked
  swing2LimitDegrees=45
  twistLimitDegrees=5
  twistMotion=Locked
 }
 {
  bone1Name=dyn_jacket_bk_r_02
  bone2Name=dyn_jacket_bk_r_01
  swing1LimitDegrees=0
  swing1Motion=Locked
  swing2LimitDegrees=45
  twistLimitDegrees=5
  twistMotion=Locked
 }
 {
  bone1Name=dyn_jacket_bk_r_03
  bone2Name=dyn_jacket_bk_r_02
  swing1LimitDegrees=0
  swing1Motion=Locked
  swing2LimitDegrees=45
  twistLimitDegrees=5
  twistMotion=Locked
 }
 {
  bone1Name=dyn_jacket_r_01
  bone2Name=dyn_jacket_r_MSTR
  swing1LimitDegrees=45
  swing1Motion=Locked
  swing2LimitDegrees=45
  twistLimitDegrees=5
  twistMotion=Locked
 }
 {
  bone1Name=dyn_jacket_r_02
  bone2Name=dyn_jacket_r_01
  swing1LimitDegrees=45
  swing1Motion=Locked
  swing2LimitDegrees=45
  twistLimitDegrees=5
  twistMotion=Locked
 }
 {
  bone1Name=dyn_jacket_r_03
  bone2Name=dyn_jacket_r_02
  swing1LimitDegrees=45
  swing1Motion=Locked
  swing2LimitDegrees=45
  twistLimitDegrees=5
  twistMotion=Locked
 }
 {bone1Name=dyn_jacket_fr_r_01, bone2Name=dyn_jacket_fr_r_MSTR, swing1LimitDegrees=45, swing2LimitDegrees=45, twistLimitDegrees=5}
 {bone1Name=dyn_jacket_fr_r_02, bone2Name=dyn_jacket_fr_r_01, swing1LimitDegrees=45, swing2LimitDegrees=45, twistLimitDegrees=5}
 {bone1Name=dyn_jacket_fr_r_03, bone2Name=dyn_jacket_fr_r_02, swing1LimitDegrees=45, swing2LimitDegrees=45, twistLimitDegrees=5}
 {
  bone1Name=spine_04
  bone2Name=spine_03
  swing1LimitDegrees=45
  swing1Motion=Free
  swing2LimitDegrees=45
  swing2Motion=Free
  twistLimitDegrees=45
  twistMotion=Free
 }
 {
  bone1Name=spine_03
  bone2Name=spine_02
  swing1LimitDegrees=45
  swing1Motion=Free
  swing2LimitDegrees=45
  swing2Motion=Free
  twistLimitDegrees=45
  twistMotion=Free
 }
 {
  bone1Name=spine_02
  bone2Name=spine_01
  swing1LimitDegrees=45
  swing1Motion=Free
  swing2LimitDegrees=45
  swing2Motion=Free
  twistLimitDegrees=45
  twistMotion=Free
 }
 {
  bone1Name=spine_01
  bone2Name=pelvis
  swing1LimitDegrees=45
  swing1Motion=Free
  swing2LimitDegrees=45
  swing2Motion=Free
  twistLimitDegrees=45
  twistMotion=Free
 }
 {
  bone1Name=thigh_l
  bone2Name=pelvis
  swing1LimitDegrees=45
  swing1Motion=Free
  swing2LimitDegrees=45
  swing2Motion=Free
  twistLimitDegrees=45
  twistMotion=Free
 }
 {
  bone1Name=thigh_r
  bone2Name=pelvis
  swing1LimitDegrees=45
  swing1Motion=Free
  swing2LimitDegrees=45
  swing2Motion=Free
  twistLimitDegrees=45
  twistMotion=Free
 }
 {
  bone1Name=dyn_jacket_bk_l_03
  bone2Name=dyn_jacket_bk_03
  swing1LimitDegrees=45
  swing1Motion=Free
  swing2LimitDegrees=45
  swing2Motion=Free
  twistLimitDegrees=45
  twistMotion=Free
 }
 {
  bone1Name=dyn_jacket_bk_r_03
  bone2Name=dyn_jacket_bk_03
  swing1LimitDegrees=45
  swing1Motion=Free
  swing2LimitDegrees=45
  swing2Motion=Free
  twistLimitDegrees=45
  twistMotion=Free
 }
 {
  bone1Name=dyn_jacket_l_03
  bone2Name=dyn_jacket_bk_l_03
  swing1LimitDegrees=45
  swing1Motion=Free
  swing2LimitDegrees=45
  swing2Motion=Free
  twistLimitDegrees=45
  twistMotion=Free
 }
 {
  bone1Name=dyn_jacket_r_03
  bone2Name=dyn_jacket_bk_r_03
  swing1LimitDegrees=45
  swing1Motion=Free
  swing2LimitDegrees=45
  swing2Motion=Free
  twistLimitDegrees=45
  twistMotion=Free
 }
 {
  bone1Name=dyn_jacket_fr_r_03
  bone2Name=dyn_jacket_r_03
  swing1LimitDegrees=45
  swing1Motion=Free
  swing2LimitDegrees=45
  swing2Motion=Free
  twistLimitDegrees=45
  twistMotion=Free
 }
 {
  bone1Name=dyn_jacket_fr_l_03
  bone2Name=dyn_jacket_l_03
  swing1LimitDegrees=45
  swing1Motion=Free
  swing2LimitDegrees=45
  swing2Motion=Free
  twistLimitDegrees=45
  twistMotion=Free
 }
]
