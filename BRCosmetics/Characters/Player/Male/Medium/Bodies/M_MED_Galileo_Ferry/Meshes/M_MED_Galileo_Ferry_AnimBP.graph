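# Secondary-motion AnimBlueprint for the Galileo Ferry male body, copying its pose from the attached parent mesh.
# Constrains the gauntlet, adds smoothed crouch offsets to jacket bones, then simulates chest and right-leg RigidBodies.
# That cached pose feeds normal and skydiving jacket simulations, blended by bIsSkyDiving before weighted constraints place the vest canister.
# The event graph sets jacket gravity: normal, or zero or upward by PawnVelocityZ while skydiving before parachute opening.
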AnimBlueprint /BRCosmetics/Characters/Player/Male/Medium/Bodies/M_MED_Galileo_Ferry/Meshes/M_MED_Galileo_Ferry_AnimBP
target_skeleton=&/M_MED_Galileo_Ferry_Skeleton
graphs [
 AnimGraph {nodes [
   n1 = Use cached pose 'jacket'()
   n2 = Use cached pose 'jacket skydiving'()
   n3 = Get bIsSkyDiving()
   n4 = "Constraint - dyn_vest_canister_in_MSTR (PPPPPPPPP)"!(
    ComponentPose=Local To Component(LocalPose=Blend Poses by bool(BlendPose_0=@n2, BlendPose_1=@n1, bActiveValue=@n3) {blendType=Linear})
    Alpha=1
    ConstraintWeights_0=0.0574
    ConstraintWeights_1:float=0.2909
    ConstraintWeights_2:float=0.0143
    ConstraintWeights_3:float=0.2202
    ConstraintWeights_4:float=0.1338
    ConstraintWeights_5:float=0.012
    ConstraintWeights_6:float=0.12
    ConstraintWeights_7:float=0.1368
    ConstraintWeights_8:float=0.0145
   ) -> (Pose) {
    alphaBoolBlend {blendOption=Linear}
    boneToModify {boneName=dyn_vest_canister_in_MSTR}
    constraintSetup [
     {targetBone {boneName=spine_03}}
     {targetBone {boneName=dyn_jacket_l_01}}
     {targetBone {boneName=dyn_jacket_l_03}}
     {targetBone {boneName=dyn_jacket_l_02}}
     {targetBone {boneName=dyn_jacket_fr_l_01}}
     {targetBone {boneName=dyn_jacket_fr_l_03}}
     {targetBone {boneName=dyn_jacket_fr_l_02}}
     {targetBone {boneName=spine_04}}
     {targetBone {boneName=spine_05}}
    ]
    constraintWeights [1,1,1,1,1,1,1,1,1]
   }
   n5 = "Constraint - dyn_vest_canister_out_MSTR (PPPPPPPP)"!(
    ComponentPose=@n4
    Alpha=1
    ConstraintWeights_0=0.0669
    ConstraintWeights_1:float=0.3241
    ConstraintWeights_2:float=0.0169
    ConstraintWeights_3:float=0.2725
    ConstraintWeights_4:float=0.0883
    ConstraintWeights_5:float=0.0105
    ConstraintWeights_6:float=0.0822
    ConstraintWeights_7:float=0.1386
   ) -> (Pose) {
    alphaBoolBlend {blendOption=Linear}
    boneToModify {boneName=dyn_vest_canister_out_MSTR}
    constraintSetup [
     {targetBone {boneName=spine_03}}
     {targetBone {boneName=dyn_jacket_l_01}}
     {targetBone {boneName=dyn_jacket_l_03}}
     {targetBone {boneName=dyn_jacket_l_02}}
     {targetBone {boneName=dyn_jacket_fr_l_01}}
     {targetBone {boneName=spine_02}}
     {targetBone {boneName=dyn_jacket_fr_l_02}}
     {targetBone {boneName=spine_04}}
    ]
    constraintWeights [1,1,1,1,1,1,1,1]
   }
   Output Pose(Result=Component To Local(ComponentPose=@n5)) {layerGroup=None}
   n6 = "Constraint - dyn_forearm_gauntlet_r (PPPPPPP)"!(
    ComponentPose=Local To Component(LocalPose=Copy Pose From Mesh() {bUseAttachedParent=true})
    Alpha=1
    ConstraintWeights_0=0.4399
    ConstraintWeights_1:float=0.0002
    ConstraintWeights_2:float=0.1133
    ConstraintWeights_3:float=0.3971
    ConstraintWeights_4:float=0.0174
    ConstraintWeights_5:float=0.0187
    ConstraintWeights_6:float=0.0134
   ) -> (Pose) {
    alphaBoolBlend {blendOption=Linear}
    boneToModify {boneName=dyn_forearm_gauntlet_r}
    constraintSetup [
     {targetBone {boneName=lowerarm_twist_02_r}}
     {targetBone {boneName=index_metacarpal_r}}
     {targetBone {boneName=lowerarm_r}}
     {targetBone {boneName=lowerarm_twist_01_r}}
     {targetBone {boneName=hand_r}}
     {targetBone {boneName=thumb_01_r}}
     {targetBone {boneName=pinky_metacarpal_r}}
    ]
    constraintWeights [1,1,1,1,1,1,1]
   }
   n7 = Get crouch alpha() -> (crouch alpha:double)
   n8 = "Transform (Modify) Bone - Bone: dyn_jacket_fr_l_MSTR"(ComponentPose=@n6, Alpha=@n7, Translation="-0.114,-2.868,1.321", Rotation="-12.939,-14.472,-59.234", Scale="1,1,1") {
    rotationMode=BMM_Additive
    rotationSpace=BCS_BoneSpace
    translationMode=BMM_Additive
    translationSpace=BCS_BoneSpace
    alphaBoolBlend {blendOption=Linear}
    alphaScaleBiasClamp {bInterpResult=true, interpSpeedDecreasing=15, interpSpeedIncreasing=15}
    boneToModify {boneName=dyn_jacket_fr_l_MSTR}
    rotation=<rot -12.939 -14.4721 -59.2339>
    translation=<-0.114 -2.868 1.321>
   }
   n9 = "Transform (Modify) Bone - Bone: dyn_jacket_l_MSTR"(ComponentPose=@n8, Alpha=@n7, Translation="0.480,-4.452,0.000", Rotation="-5.165,-18.684,-14.360", Scale="1,1,1") {
    rotationMode=BMM_Additive
    rotationSpace=BCS_BoneSpace
    translationMode=BMM_Additive
    translationSpace=BCS_BoneSpace
    alphaBoolBlend {blendOption=Linear}
    alphaScaleBiasClamp {bInterpResult=true, interpSpeedDecreasing=15, interpSpeedIncreasing=15}
    boneToModify {boneName=dyn_jacket_l_MSTR}
    rotation=<rot -5.165 -18.684 -14.3601>
    translation=<0.48 -4.452>
   }
   n10 = RigidBody(ComponentPose=@n9) {
    bClampLinearTranslationLimitToRefPose=true
    bOverrideWorldGravity=true
    overridePhysicsAsset=&/M_MED_Galileo_Ferry_Dyn_chest
    simulationSpace=BaseBoneSpace
    alphaBoolBlend {blendOption=Linear}
    baseBoneRef {boneName=spine_05}
    componentAppliedLinearAccClamp=<3000 3000 3000>
    componentLinearAccScale=<0.4 0.3 0.3>
    componentLinearVelScale=<0.4 0.3 0.3>
    overrideWorldGravity {z=-980}
   }
   n11 = RigidBody(ComponentPose=@n10) {
    bClampLinearTranslationLimitToRefPose=true
    bOverrideWorldGravity=true
    overridePhysicsAsset=&/M_MED_Galileo_Ferry_Dyn_leg_r
    simulationSpace=BaseBoneSpace
    alphaBoolBlend {blendOption=Linear}
    baseBoneRef {boneName=thigh_r}
    componentAppliedLinearAccClamp=<3000 3000 3000>
    componentLinearAccScale=<0.4 0.3 0.3>
    componentLinearVelScale=<0.4 0.3 0.3>
    overrideWorldGravity {z=-980}
   }
   Save cached pose 'initial Pose'(Pose=Component To Local(ComponentPose=@n11)) {cacheName=initial Pose}
   n12 = RigidBody!(
    ComponentPose=Local To Component(LocalPose=Use cached pose 'initial Pose'())
    Alpha
    OverrideWorldGravity:Vector=Get gravity_override() -> (gravity_override:Vector)
    ExternalForce
   ) -> (Pose) {
    bClampLinearTranslationLimitToRefPose=true
    bOverrideWorldGravity=true
    overridePhysicsAsset=&/M_MED_Galileo_Ferry_Dyn_jacket
    simulationSpace=BaseBoneSpace
    alphaBoolBlend {blendOption=Linear}
    baseBoneRef {boneName=spine_02}
    componentAppliedLinearAccClamp=<3000 3000 3000>
    componentLinearAccScale=<0.4 0.1 0.1>
    componentLinearVelScale=<0.4 0.1 0.1>
    overrideWorldGravity {z=-980}
   }
   Save cached pose 'jacket'(Pose=Component To Local(ComponentPose=@n12)) {cacheName=jacket}
   n13 = RigidBody(ComponentPose=Local To Component(LocalPose=Use cached pose 'initial Pose'()), OverrideWorldGravity=Get gravity_override()) {
    bClampLinearTranslationLimitToRefPose=true
    bOverrideWorldGravity=true
    overridePhysicsAsset=&/M_MED_Galileo_Ferry_Dyn_jacket_skydiving
    simulationSpace=BaseBoneSpace
    alphaBoolBlend {blendOption=Linear}
    baseBoneRef {boneName=spine_02}
    componentAppliedLinearAccClamp=<3000 3000 3000>
    componentLinearAccScale=<0.4 0.1 0.1>
    componentLinearVelScale=<0.4 0.1 0.1>
    overrideWorldGravity {z=-980}
   }
   Save cached pose 'jacket skydiving'(Pose=Component To Local(ComponentPose=@n13)) {cacheName=jacket skydiving}
  ]}
 EventGraph {nodes [
   Event BlueprintUpdateAnimation()
   > n1 = Branch(Condition=Get bIsSkyDiving())
   > n2 = Branch(Condition=Get bIsParachuteOpened())
   Set gravity_override(execute=[@n1.else, @n2.then], gravity_override="0, 0,-980")
   n3 = Branch(execute=@n2.else, Condition="float > float"(A=Get PawnVelocityZ(), B=50) {is=title K2Node_CallFunction})
   > Set gravity_override(gravity_override="0, 0,400")
   Set gravity_override(execute=@n3.else, gravity_override="0, 0, 0")
   Event OnBeginCrouching()
   > Set crouch alpha(crouch alpha:double=1) -> (Output_Get:double)
   Event OnEndCrouching()
   > Set crouch alpha(crouch alpha=0)
  ]}
]
variables [
 crouch alpha {type=number}
 gravity_override {type=Vector}
]
